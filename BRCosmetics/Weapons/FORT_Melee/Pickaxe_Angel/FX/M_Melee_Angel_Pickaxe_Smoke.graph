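# Translucent unlit smoke material for the Angel pickaxe's effects, usable on Niagara sprites, ribbons, meshes and beam trails.
# By default, three panning, partly noise-distorted textures form opacity, shaped by sine UV edge masks, depth fade and a sphere mask, colored by a four-color U gradient.
# A static switch swaps in particle-tinted flipbook smoke masked around the actor.
# Other switches add time-of-day alpha and panning-normal vertex offset that grows away from the object.
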Material /BRCosmetics/Weapons/FORT_Melee/Pickaxe_Angel/FX/M_Melee_Angel_Pickaxe_Smoke
graph {
 properties {
  bUsedWithBeamTrails=true
  bUsedWithNiagaraMeshParticles=true
  bUsedWithNiagaraRibbons=true
  bUsedWithNiagaraSprites=true
  blendMode=BLEND_Translucent
  pixelDepthOffsetMode=PDOM_Legacy
  shadingModel=MSM_Unlit
  usedShadingModels=Unlit
 }
 nodes [
  ActorPositionWS()
  n1 = ParticleColor()
  n2 = TextureSampleParameter2D() {parameterName=BaseTexture, texture=/BRCosmetics/Athena/Effects/Textures/T_InkySmoke_Flipbook}
  n3 = LinearInterpolate(
   A=LinearInterpolate(A=@n1.RGB, B=Multiply(A=@n1.RGB, B=@n1.RGB), Alpha=ScalarParameter() {parameterName=SaturationBlend})
   B=@n1.RGB
   Alpha=@n2.G
  )
  n4 = ScalarParameter() {defaultValue=1, parameterName=Brightness}
  n5 = VectorParameter() {parameterName=color_01, defaultValue=<rgba 0.0615 0.8602 1 1>}
  n6 = VectorParameter() {parameterName=color_02, defaultValue=<rgba 0 0.137 1 1>}
  n7 = Add(A=MaterialFunctionCall(Float2=TextureCoordinate()) {materialFunction=&BreakOutFloat2Components}.R) {constB=-0.2}
  n8 = TextureCoordinate()
  n9 = LinearInterpolate(
   A=LinearInterpolate(
    A=LinearInterpolate(A=@n5.RGB, B=@n6.RGB, Alpha=@n7)
    B=VectorParameter() {parameterName=color_03, defaultValue=<rgba 1 0.1825 0.1228 1>}.RGB
    Alpha=@n7
   )
   B=Multiply(A=VectorParameter() {parameterName=color_04, defaultValue=<rgba 1 0.3154 0.0118 1>}.RGB)
   Alpha=Add(A=MaterialFunctionCall(Float2=@n8) {materialFunction=&BreakOutFloat2Components}.R) {constB=-0.4}
  )
  n10 = ParticleColor()
  MaterialFunctionCall(Emissive=StaticSwitchParameter(True=Multiply(A=@n3, B=@n4), False=Multiply(A=@n9, B=@n10.RGB)) {parameterName=Use Trail mask particles}) {materialFunction=&MF_FX_EmissiveAdjustments}
  n11 = Panner(Coordinate=TextureCoordinate() {uTiling=0.35, vTiling=0.3}) {speedX=-0.3}
  n12 = AppendVector(
   A=Constant()
   B=ComponentMask(None=TextureSample(UVs=@n11) {samplerType=SAMPLERTYPE_Normal, texture=&T_Weap_Distort_N_02}.RGB) {b=true, r=true}
  )
  n13 = ScalarParameter() {defaultValue=45, parameterName=WPO}
  n14 = SphereMask(
   A=WorldPosition().XYZ
   B=ObjectPositionWS()
   Radius=ScalarParameter() {defaultValue=130, parameterName=WPO Range}
   Hardness=Constant() {r=20}
  )
  StaticSwitchParameter(True=Multiply(A=Multiply(A=@n12, B=@n13), B=OneMinus(None=@n14)), False=Constant()) {parameterName=Be use for WPO}
  n15 = Multiply(A=@n1.A, B=@n2.G)
  n16 = ScalarParameter() {defaultValue=20, parameterName=FadeDist}
  n17 = WorldPosition()
  n18 = ActorPositionWS()
  n19 = ScalarParameter() {defaultValue=350, parameterName=Trail particles mask}
  n20 = Multiply(
   A=TextureCoordinate()
   B=AppendVector(
    A=ScalarParameter() {defaultValue=0.3, parameterName=Texture_C_UTiling}
    B=ScalarParameter() {defaultValue=1, parameterName=Texture_C_VTiling}
   )
  )
  n21 = Multiply(A=TextureSampleParameter2D(UVs=Panner(Coordinate=@n20) {speedX=-0.7}) {parameterName=Texture_C, texture=/Game/Effects/Fort_Effects/Textures/Smoke/T_Smoke_Trail_Soft_Chromatic}.R)
  n22 = Multiply(A=LinearInterpolate(A=Round(None=@n21), B=@n21, Alpha=ScalarParameter() {defaultValue=0.5, parameterName=Texture_C_Lerp})) {constB=0.5}
  n23 = TextureCoordinate()
  n24 = AppendVector(
   A=ScalarParameter() {defaultValue=0.7, parameterName=Texture_A_UTiling}
   B=ScalarParameter() {defaultValue=1, parameterName=Texture_A_VTiling}
  )
  n25 = TextureSampleParameter2D(UVs=Panner(Coordinate=Multiply(A=@n23, B=@n24)) {speedX=-1}) {parameterName=Texture_A, samplerType=SAMPLERTYPE_Grayscale, texture=&T_Smoke_Trail_Soft}
  n26 = Multiply(A=@n25.R)
  n27 = LinearInterpolate(A=Round(None=@n26), B=@n26, Alpha=ScalarParameter() {defaultValue=0.5, parameterName=Texture_A_Lerp})
  n28 = ScalarParameter() {defaultValue=0.3, parameterName=Texture_B_UTiling}
  n29 = ScalarParameter() {defaultValue=1, parameterName=Texture_B_VTiling}
  n30 = Panner(Coordinate=TextureCoordinate()) {speedX=-1, speedY=1}
  n31 = Add(
   A=Multiply(A=TextureCoordinate(), B=AppendVector(A=@n28, B=@n29))
   B=Multiply(
    A=TextureSample(UVs=@n30) {samplerType=SAMPLERTYPE_Grayscale, texture=&LowResBlurredNoise}.R
    B=ScalarParameter() {parameterName=Noise}
   )
  )
  n32 = Multiply(
   A=TextureSampleParameter2D(UVs=Panner(Coordinate=@n31) {speedX=-0.4}) {parameterName=Texture_B, texture=/Game/Effects/Fort_Effects/Textures/Energy/T_EnergyFlow_Packed}.R
   B=ScalarParameter() {defaultValue=1, parameterName=Texture_B_Intensity}
  )
  n33 = Round(None=@n32)
  n34 = ScalarParameter() {defaultValue=1, parameterName=Texture_B_Lerp}
  n35 = LinearInterpolate(
   A=Multiply(A=LinearInterpolate(A=@n22, B=@n27)) {constB=1.2}
   B=Multiply(A=LinearInterpolate(A=@n33, B=@n32, Alpha=@n34) {constAlpha=1}) {constB=0.8}
  )
  n36 = MaterialFunctionCall(Float2=TextureCoordinate() {vTiling=0.99}) {materialFunction=&BreakOutFloat2Components}
  n37 = Power(Base=Sine(None=@n36.R) {period=2}) {constExponent=1.5}
  n38 = Power(Base=Sine(None=@n36.G) {period=2}) {constExponent=1}
  n39 = Multiply(A=@n10.A, B=Multiply(A=Multiply(A=@n35) {constB=1.5}, B=Multiply(A=@n37, B=@n38)))
  n40 = DepthFade(Opacity=@n39, FadeDistance=ScalarParameter() {defaultValue=20, parameterName=Depth_Fade})
  n41 = Multiply(
   A=ShadingPathSwitch(Default=@n40, Deferred=@n40, Mobile=@n39) {desc=Migrated from FeatureLevelSwitch}
   B=SphereMask(A=WorldPosition().XYZ, B=ObjectPositionWS(), Radius=ScalarParameter() {defaultValue=130, parameterName=Radius}, Hardness=Constant())
  )
  n42 = Multiply(A=@n41, B=ScalarParameter() {defaultValue=1, parameterName=color_day})
  n43 = Multiply(A=@n41, B=ScalarParameter() {defaultValue=1, parameterName=color_night})
  StaticSwitchParameter(
   True=Multiply(A=DepthFade(Opacity=@n15, FadeDistance=@n16), B=SphereMask(A=@n17.XYZ, B=@n18, Radius=@n19) {attenuationRadius=350, hardnessPercent=0})
   False=StaticSwitchParameter(
    True=MaterialFunctionCall(Color_Morning=@n42, Color_Day=@n42, Color_Evening=@n42, Color_Night=@n43) {materialFunction=&Color_Time_of_Day}.Result
    False=@n41
   ) {parameterName=Use Alpha color time of day}
  ) {parameterName=Use Trail mask particles}
 ]
}
