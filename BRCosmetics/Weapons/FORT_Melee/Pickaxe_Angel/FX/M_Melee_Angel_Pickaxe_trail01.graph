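# Translucent unlit trail material for the Angel pickaxe, usable on beam trails and Niagara ribbons or meshes.
# Blends two panning smoke textures with adjustable hard rounding; a dynamic parameter can offset the second.
# Opacity combines that noise with sine edge fades, particle alpha, time-of-day strength, non-mobile depth fade and optional Fresnel.
# Color is a switchable four-color gradient along U times particle color, with optional day/night tint, feeding an emissive adjustment function.
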Material /BRCosmetics/Weapons/FORT_Melee/Pickaxe_Angel/FX/M_Melee_Angel_Pickaxe_trail01
graph {
 properties {
  bUsedWithBeamTrails=true
  bUsedWithNiagaraMeshParticles=true
  bUsedWithNiagaraRibbons=true
  blendMode=BLEND_Translucent
  pixelDepthOffsetMode=PDOM_Legacy
  shadingModel=MSM_Unlit
  usedShadingModels=Unlit
 }
 nodes [
  n1 = ParticleColor()
  n2 = Multiply(
   A=TextureCoordinate()
   B=AppendVector(
    A=ScalarParameter() {defaultValue=0.5, parameterName=Texture_A_UTiling}
    B=ScalarParameter() {defaultValue=1, parameterName=Texture_A_VTiling}
   )
  )
  n3 = Multiply(A=Time(), B=ScalarParameter() {defaultValue=1, parameterName=Texture_A_speed})
  n4 = Multiply(
   A=TextureSampleParameter2D(UVs=Panner(Coordinate=@n2, Time=@n3) {speedX=1}) {parameterName=Texture_A, texture=/Game/Effects/Fort_Effects/Textures/Smoke/T_Smoke_Trail_Soft_2}.R
   B=ScalarParameter() {defaultValue=1, parameterName=Texture_A_Intensity}
  )
  n5 = TextureCoordinate()
  n6 = DynamicParameter() {defaultValue=<rgba 0 0 1 1>, paramNames [U Offset, V Offset, Param3, Param4]}
  n7 = LinearInterpolate(A=@n5, B=Add(A=@n5, B=AppendVector(A=@n6.U Offset, B=@n6.U Offset)), Alpha=ScalarParameter() {parameterName=Texture_B_UseOffset})
  n8 = AppendVector(
   A=ScalarParameter() {defaultValue=2, parameterName=Texture_B_UTiling}
   B=ScalarParameter() {defaultValue=0.7, parameterName=Texture_B_VTiling}
  )
  n9 = Time()
  n10 = ScalarParameter() {defaultValue=1, parameterName=Texture_B_speed}
  n11 = TextureSampleParameter2D(UVs=Panner(Coordinate=Multiply(A=@n7, B=@n8), Time=Multiply(A=@n9, B=@n10)) {speedX=0.7}) {parameterName=Texture_B, samplerType=SAMPLERTYPE_Grayscale, texture=&T_Smoke_Trail_Soft}
  n12 = Multiply(A=@n11.R, B=ScalarParameter() {defaultValue=1.5, parameterName=Texture_B_Intensity}) {constB=1.5}
  n13 = Power(Base=@n12, Exp=ScalarParameter() {defaultValue=1, parameterName=Texture_Power_B})
  n14 = Round(None=@n12)
  n15 = LinearInterpolate(
   A=LinearInterpolate(A=Round(None=@n4), B=@n4, Alpha=ScalarParameter() {defaultValue=1, parameterName=Texture_A_Lerp}) {constAlpha=1}
   B=LinearInterpolate(
    A=StaticSwitchParameter(True=@n13, False=@n14) {parameterName=Use Round}
    B=@n12
    Alpha=ScalarParameter() {defaultValue=0.3, parameterName=Texture_B_Lerp}
   ) {constAlpha=0.3}
  )
  n16 = Multiply(A=@n15, B=ScalarParameter() {defaultValue=1, parameterName=Alpha_day})
  n17 = Multiply(A=@n15, B=ScalarParameter() {defaultValue=0.7, parameterName=Alpha_night}) {constB=0.7}
  n18 = MaterialFunctionCall(Float2=TextureCoordinate() {vTiling=0.99}) {materialFunction=&BreakOutFloat2Components}
  n19 = Sine(None=@n18.R) {period=2}
  n20 = ScalarParameter() {defaultValue=2, parameterName=HorizontalFade_Power}
  n21 = Saturate(None=Multiply(A=Power(Base=@n19, Exp=@n20), B=ScalarParameter() {defaultValue=1, parameterName=HorizontalFade_Multiplier}))
  n22 = Power(Base=Sine(None=@n18.G) {period=2}) {constExponent=2.5}
  n23 = Multiply(
   A=@n1.A
   B=Multiply(
    A=MaterialFunctionCall(Color_Morning=@n16, Color_Day=@n16, Color_Evening=@n16, Color_Night=@n17) {materialFunction=&Color_Time_of_Day}.Result
    B=Multiply(A=@n21, B=@n22)
   )
  )
  n24 = DepthFade(Opacity=@n23, FadeDistance=ScalarParameter() {defaultValue=20, parameterName=Depth_Fade})
  n25 = ShadingPathSwitch(Default=@n24, Deferred=@n24, Mobile=@n23) {desc=Migrated from FeatureLevelSwitch}
  n26 = MaterialFunctionCall(
   Brightness=ScalarParameter() {defaultValue=1, group=Alpha, parameterName="CameraNormal_Fresnel_Brightness (S)"}
   Exponent=ScalarParameter() {defaultValue=3, group=Alpha, parameterName="CameraNormal_Fresnel_Exponent (S)"}
   Vertex / Pixel Normal=VertexNormalWS()
  ) {materialFunction=&MF_Fresnel_FX}
  LinearInterpolate(
   A=@n25
   B=Multiply(A=@n25, B=Multiply(A=@n26.Fresnel Double Sided))
   Alpha=ScalarParameter() {group=Alpha, parameterName=Use_CameraNormal_Fresnel}
  )
  n27 = Add(A=MaterialFunctionCall(Float2=TextureCoordinate()) {materialFunction=&BreakOutFloat2Components}.R) {constB=-0.2}
  n28 = LinearInterpolate(
   A=LinearInterpolate(
    A=VectorParameter() {parameterName=color_01, defaultValue=<rgba 0.0615 0.8602 1 1>}.RGB
    B=VectorParameter() {parameterName=color_02, defaultValue=<rgba 0 0.137 1 1>}.RGB
    Alpha=@n27
   )
   B=VectorParameter() {parameterName=color_03, defaultValue=<rgba 1 0.1825 0.1228 1>}.RGB
   Alpha=@n27
  )
  n29 = Multiply(A=VectorParameter() {parameterName=color_04, defaultValue=<rgba 1 0.3154 0.0118 1>}.RGB)
  n30 = Add(A=MaterialFunctionCall(Float2=TextureCoordinate()) {materialFunction=&BreakOutFloat2Components}.R) {constB=-0.4}
  n31 = StaticSwitchParameter(True=Multiply(A=LinearInterpolate(A=@n28, B=@n29, Alpha=@n30), B=@n1.RGB), False=@n1.RGB) {defaultValue=true, parameterName=Use color lerp}
  n32 = Multiply(A=VectorParameter() {parameterName=Day color, defaultValue=<rgba 1 1 1 0>}.RGB, B=@n31)
  n33 = Multiply(A=VectorParameter() {parameterName=Night color, defaultValue=<rgba 1 1 1 0>}.RGB, B=@n31)
  MaterialFunctionCall(Emissive=StaticSwitchParameter(
    True=MaterialFunctionCall(Color_Morning=@n32, Color_Day=@n32, Color_Evening=@n32, Color_Night=@n33) {materialFunction=&Color_Time_of_Day}.Result
    False=@n31
   ) {parameterName=Use color time of day}) {materialFunction=&MF_FX_EmissiveAdjustments}
 ]
}
